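; Blueprint controller for the Cartmanland shooting gallery music in the MusicParty audio setup.
; A custom StartStopShootingGalleryMusic event takes a boolean and fires a trigger parameter on the Audio component.
; The boolean selects the trigger name, "StartGame" or "StopGame", leaving the music response to the audio asset.
; The construction script only calls its parent.
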(event Custom|StartStopShootingGalleryMusic
  (params (bShouldStartMusic :type Boolean))
  (Audio|Parameter|ExecuteTriggerParameter (Variables|Default|GetAudio) (Utilities|String|SelectName "StartGame" "StopGame" bShouldStartMusic)))

(fn ConstructionScript
  (|CallParentFunction))
