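# Mono looping MetaSound source for the Shooting Gallery's Cartmanland music, switching between idle and game playback.
# On play, an idle wave starts at a random point within its duration.
# StartGame plays one wave immediately and another after a short delay, while a BPM-timed interpolation crossfades idle against the delayed wave; StopGame reverses the fade.
# Base submix and submix sends are off, with output sent to the AB_ShootingGallery audio bus.
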MetaSoundSource /ShootingGallery/Audio/ShootingGalleryMusic/MSS_ShootingGalleryMusic_Cartmanland
inputs [
 StartGame {type=Trigger}
 StopGame {type=Trigger}
 UE.Source.OnPlay {type=Trigger}
]
interfaces [UE.OutputFormat.Mono:1.0, UE.Source:1.1]
outputs ["UE.OutputFormat.Mono.Audio:0" {type=Audio}]
pages [Default {
  input_defaults {StartGame=null, StopGame=null, UE.Source.OnPlay=null}
  nodes [
   n1 = VariableAccessor.WaveAsset(Variable=InitVariable.WaveAsset().Variable)
   GetWaveDuration_0 = "UE.Get Wave Duration: "(Wave=@n1.Value)
   n2 = UE.RandomTime(Next=^UE.Source.OnPlay, Max=@GetWaveDuration_0.Duration)
   WavePlayer_0 = UE.Wave Player:Mono(Play=@n2.On Next, Start Time=@n2.Value, Wave Asset=@n1.Value) {also {defaults {Loop="<MetaSoundBuilderResult.SUCCEEDED: 0>"}, version [1,1]}}
   AudioMixerAudioMixerMono2_0 = "AudioMixer.Audio Mixer (Mono, 2)"(In 0=@WavePlayer_0.Out Mono) {also {defaults {Gain 0=2}}}
   TriggerRouteTriggerRouteFloat2_0 = "TriggerRoute.Trigger Route (Float, 2):Float"(Set 1=^StartGame, Set 0=^StopGame) {also {defaults {Value 1=1}}}
   n3 = UE.InterpTo:Audio(Interp Time=UE.BPMToSeconds().Seconds, Target=@TriggerRouteTriggerRouteFloat2_0.Value)
   TriggerDelay_0 = UE.Trigger Delay(In=^StartGame) {also {defaults {Delay Time=0.827}}}
   WavePlayer_1 = UE.Wave Player:Mono(Play=@TriggerDelay_0.Out, Wave Asset=VariableAccessor.WaveAsset(Variable=InitVariable.WaveAsset().Variable).Value) {also {version [1,1]}}
   WavePlayer_2 = UE.Wave Player:Mono(Play=^StartGame, Wave Asset=VariableAccessor.WaveAsset(Variable=InitVariable.WaveAsset().Variable).Value) {also {version [1,1]}}
   AudioMixerAudioMixerMono3_0 = "AudioMixer.Audio Mixer (Mono, 3)"(
    In 2=@AudioMixerAudioMixerMono2_0.Out
    Gain 0=@n3.Value
    Gain 2=UE.Subtract:Float(AdditionalOperands=@n3.Value) {also {defaults {PrimaryOperand=1}}}.Out
    In 0=@WavePlayer_1.Out Mono
    In 1=@WavePlayer_2.Out Mono
   )
  ]
  links [{dst="@UE.OutputFormat.Mono.Audio:0", src=AudioMixerAudioMixerMono3_0.Out}]
 }]
properties {
 assetImportData=&self:AssetImportData
 bEnableBaseSubmix=false
 bEnableSubmixSends=false
 bLooping=true
 duration=10000
 sampleRate=48000
 soundClassObject=/Game/Sounds/FortSoundClassesAndModes/Fort_Music_Gameplay
 virtualizationMode=PlayWhenSilent
 volume=3
 busSends [{audioBus=&/AB_ShootingGallery, sendLevel=1}]
}
